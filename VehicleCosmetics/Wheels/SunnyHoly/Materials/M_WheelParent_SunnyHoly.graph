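# Masked parent material for SunnyHoly vehicle wheels, using vertex color red to split rim and tire.
# The rim blends sun and custom colors through packed masks with Fresnel glow; the tire adds Fresnel and detail glow over two colors mixed by panning noise.
# Both animate World Position Offset from object-relative position plus a sine on panned UV1, with per-part multipliers.
# The attributes then pass through vehicle damage and de-rez transition functions.
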Material /VehicleCosmetics/Wheels/SunnyHoly/Materials/M_WheelParent_SunnyHoly
graph {
 properties {
  bAutomaticallySetUsageInEditor=false
  bUseMaterialAttributes=true
  bUsedWithSkeletalMesh=true
  blendMode=BLEND_Masked
  usedShadingModels=Default Lit
 }
 nodes [
  n1 = TextureSampleParameter2D() {parameterName=Rim_Sunny_N, samplerType=SAMPLERTYPE_Normal, texture=/VehicleCosmetics/Wheels/SunnyHoly/Textures/T_SunnyHoly_Rim_N}
  n2 = MaterialFunctionCall!(Input Normal=@n1.RGB, Input Sharpness, Input Two Sided) {materialFunction=/VehicleCosmetics/Wheels/CarCar_Holographic/Materials/MF_Fresnel_Advanced}
  n3 = ScalarParameter() {defaultValue=1, parameterName=Fresnel_Glow_Intencity}
  n4 = Multiply(A=@n3, B=VectorParameter() {parameterName=Fresnel_Tint, defaultValue=<rgba 0.87 0.35 0.09 1>}.RGB)
  n5 = VectorParameter() {parameterName=Sun_Outer_Color, defaultValue=<rgba 0.93 0.7 0.5 1>}
  n6 = TextureSampleParameter2D() {parameterName=Rim_RGB, samplerType=SAMPLERTYPE_LinearColor, texture=/VehicleCosmetics/Wheels/SunnyHoly/Textures/T_SunnyHoly_Rim_RGB}
  n7 = Multiply(
   A=Multiply(
    A=VectorParameter() {parameterName=Custom_Color, defaultValue=<rgba 0.64 0.26 0.07 1>}.RGB
    B=ScalarParameter() {defaultValue=0.5, parameterName=Custom_Color_Intensity}
   )
   B=@n6.A
  )
  n8 = Multiply(A=Add(A=Multiply(A=@n2.Edge, B=@n4), B=LinearInterpolate(A=@n5.RGB, B=@n7, Alpha=@n6.R)), B=Add(A=@n6.R, B=@n6.R))
  n9 = TextureSampleParameter2D() {parameterName=Tire_Sunny_N, samplerType=SAMPLERTYPE_Normal, texture=/VehicleCosmetics/Wheels/SunnyHoly/Textures/T_SunnyHoly_Tire_N2}
  n10 = ScalarParameter() {defaultValue=1.5, parameterName=Fresnel_Sharpness}
  n11 = VectorParameter() {parameterName=Tire_Fresnel_Tint, defaultValue=<rgba 0.18 0.45 0.91 1>}
  n12 = TextureSampleParameter2D() {parameterName=Tire_RGB, samplerType=SAMPLERTYPE_LinearColor, texture=/VehicleCosmetics/Wheels/SunnyHoly/Textures/T_SunnyHoly_Tire_RGB}
  n13 = ScalarParameter() {defaultValue=0.8, parameterName=Tire_Detail_Glow_Intencity}
  n14 = Multiply(
   A=VectorParameter() {parameterName=Tire_Custom_Color, defaultValue=<rgba 0 0.3 0.79 1>}.RGB
   B=ScalarParameter() {defaultValue=0.5, parameterName=Custom_Color_Intensity}
  )
  n15 = Multiply(
   A=VectorParameter() {parameterName=Secondary_Color, defaultValue=<rgba 0 0.37 0.97 1>}.RGB
   B=ScalarParameter() {defaultValue=0.7, parameterName=Secondary_Color_Glow_Intensity}
  )
  n16 = TextureCoordinate() {uTiling=2, vTiling=2}
  n17 = Multiply(
   A=TextureSample(UVs=Panner() {speedX=0.1, speedY=0.05}) {texture=/VehicleCosmetics/UniversalSkins/AmericanFootball/Universal_Noise_67_RGB}.R
   B=TextureSample(UVs=Panner(Coordinate=@n16) {speedX=0.02, speedY=-0.05}) {texture=&Universal_Noise_67_RGB}.A
  )
  n18 = MaterialFunctionCall(In=Add(A=@n17, B=@n17), Contrast=Constant()) {materialFunction=&CheapContrast}
  n19 = Add(
   A=Multiply(
    A=MaterialFunctionCall(Input Normal=@n9.RGB, Input Sharpness=@n10) {materialFunction=&MF_Fresnel_Advanced}.Edge
    B=Multiply(A=@n3, B=@n11.RGB)
   )
   B=Add(A=Multiply(A=@n12.B, B=@n13), B=LinearInterpolate(A=@n14, B=@n15, Alpha=@n18.Result))
  )
  n20 = VertexColor()
  n21 = Multiply(A=@n5.RGB, B=ScalarParameter() {defaultValue=2, parameterName=Sun_Outer_Glow_Intensity})
  n22 = TextureSampleParameter2D() {parameterName=Alpha_LOD, texture=/VehicleCosmetics/Bodies/Octane/Skins/Octane_Omens/T_White}
  n23 = ScalarParameter() {parameterName=Masked}
  n24 = VectorParameter() {parameterName=WPO_Rim_Multiplier, defaultValue=<rgba 1 1 1 1>}
  n25 = WorldPosition()
  n26 = ObjectPositionWS()
  n27 = Multiply(A=Multiply(A=ObjectRadius(), B=Subtract(A=@n25.XYZ, B=@n26)), B=Constant() {r=-1.9})
  n28 = Constant() {r=-2}
  n29 = Panner(Coordinate=TextureCoordinate() {coordinateIndex=1}) {speedX=1, speedY=1}
  n30 = Multiply(A=Sine(None=ComponentMask(None=@n29) {r=true}), B=Constant() {r=400})
  n31 = Constant3Vector() {constant=<rgba 1 1 1 0>}
  n32 = Divide(A=Add(A=Multiply(A=@n27, B=@n28), B=Multiply(A=@n30, B=@n31)), B=Constant() {r=3100})
  n33 = VectorParameter() {parameterName=WPO_Tire_Multiplier, defaultValue=<rgba 1 1 1 1>}
  n34 = SetMaterialAttributes!(
   MaterialAttributes
   Base Color=LinearInterpolate(A=@n8, B=@n19, Alpha=@n20.R)
   Emissive Color=LinearInterpolate(A=LinearInterpolate(A=@n8, B=@n21, Alpha=@n6.G), B=@n19, Alpha=@n20.R)
   Opacity=LinearInterpolate(A=LinearInterpolate(A=@n22.RGB, Alpha=@n23), Alpha=@n20.R)
   World Position Offset=LinearInterpolate(A=Multiply(A=@n24.RGB, B=@n32), B=Multiply(A=@n32, B=@n33.RGB), Alpha=@n20.R)
  ) {attributeSetTypes [
    #1
    #2
    #3
    #4
   ]}
  MaterialFunctionCall(In=MaterialFunctionCall(Material=@n34) {materialFunction=&MF_BaseVehicle_Damage}.Result) {materialFunction=&MF_VehicleDeRez_Transition}
  OneMinus(None=@n20.R)
 ]
}
