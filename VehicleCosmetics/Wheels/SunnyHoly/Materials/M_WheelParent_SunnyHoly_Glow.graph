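# Parent material for the glow on the SunnyHoly vehicle wheel: translucent, unlit, two-sided, with a flat tinted emissive.
# Opacity combines two panning caustics samples, masked by a gradient texture with a faint base level.
# World position offset pushes vertices from the object centre with a UV1-driven sine wobble, blending rim and tire multipliers by vertex red.
# The attributes pass through the shared vehicle damage and de-rez transition functions.
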Material /VehicleCosmetics/Wheels/SunnyHoly/Materials/M_WheelParent_SunnyHoly_Glow
graph {
 properties {
  bAutomaticallySetUsageInEditor=false
  bUseMaterialAttributes=true
  bUsedWithSkeletalMesh=true
  blendMode=BLEND_Translucent
  shadingModel=MSM_Unlit
  twoSided=true
  usedShadingModels=Unlit
 }
 nodes [
  n1 = Multiply(
   A=VectorParameter() {parameterName=CustomColor, defaultValue=<rgba 1 0.19 0 1>}.RGB
   B=ScalarParameter() {defaultValue=32, parameterName=Glow_EmissiveMultioly}
  )
  n2 = TextureCoordinate()
  n3 = AppendVector(
   A=Multiply(A=ComponentMask(None=@n2) {r=true}, B=ScalarParameter() {defaultValue=1, parameterName=Glow_TileX})
   B=Multiply(A=ComponentMask(None=@n2) {g=true}, B=ScalarParameter() {defaultValue=1, parameterName=Glow_TileY})
  )
  n4 = Multiply(A=Time(), B=ScalarParameter() {defaultValue=1, parameterName=Glow_PanSpeedAdjast})
  n5 = TextureSampleParameter2D(UVs=Panner(Coordinate=@n3, Time=@n4) {speedX=0.0125, speedY=1}) {parameterName=Caustics_D, texture=/VehicleCosmetics/Wheels/_SharedTextures/FX/Caustics_D}
  n6 = TextureSampleParameter2D(UVs=Panner(Coordinate=@n3, Time=@n4) {speedX=0.075, speedY=0.067}) {parameterName=Caustics_D, texture=&Caustics_D}
  n7 = Multiply(A=@n6.G, B=@n6.G)
  n8 = TextureSampleParameter2D(UVs=TextureCoordinate()) {
   parameterName=Glow_Gradient_Texture
   samplerType=SAMPLERTYPE_LinearColor
   texture=/VehicleCosmetics/Wheels/Twist_Twist/Textures/Gradient_V2_D
  }
  n9 = Add(A=Multiply(A=Add(A=@n5.R, B=@n7), B=@n8.R), B=Multiply(A=@n8.R, B=Constant() {r=0.025}))
  n10 = WorldPosition()
  n11 = ObjectPositionWS()
  n12 = Multiply(A=Multiply(A=ObjectRadius(), B=Subtract(A=@n10.XYZ, B=@n11)), B=Constant() {r=-2})
  n13 = Constant() {r=-2}
  n14 = Panner(Coordinate=TextureCoordinate() {coordinateIndex=1}) {speedX=1, speedY=1}
  n15 = Multiply(A=Sine(None=ComponentMask(None=@n14) {r=true}), B=Constant() {r=300})
  n16 = Constant3Vector() {constant=<rgba 1 1 1 0>}
  n17 = Divide(A=Add(A=Multiply(A=@n12, B=@n13), B=Multiply(A=@n15, B=@n16)), B=Constant() {r=3100})
  n18 = LinearInterpolate(
   A=Multiply(A=ScalarParameter() {defaultValue=0.1, parameterName=WPO_Rim_Multiplier}, B=@n17)
   B=Multiply(A=@n17, B=VectorParameter() {parameterName=WPO_Tire_Multiplier, defaultValue=<rgba 1 1 1 1>}.RGB)
   Alpha=VertexColor().R
  )
  MaterialFunctionCall(In=MaterialFunctionCall(Material=SetMaterialAttributes!(MaterialAttributes, Emissive Color=@n1, Opacity=@n9, World Position Offset=@n18) {attributeSetTypes [#1, #2, #3]}) {materialFunction=&MF_BaseVehicle_Damage}.Result) {materialFunction=&MF_VehicleDeRez_Transition}
 ]
}
